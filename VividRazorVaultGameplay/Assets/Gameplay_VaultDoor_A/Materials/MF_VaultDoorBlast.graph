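# Material function for the vault door that optionally overlays a blast decal on incoming material attributes via the UseBlastDecal static switch.
# The decal samples UV channel 2 with offset and centered scaling, clamped or wrapped by static switch; ScaleU drives both axes, leaving ScaleV unused.
# Decal alpha, times blend parameters, lerps BaseColor toward itself multiplied by decal RGB and squared BlastColor, and Specular toward BlastSpecular.
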MaterialFunction /VividRazorVaultGameplay/Assets/Gameplay_VaultDoor_A/Materials/MF_VaultDoorBlast
graph {nodes [
  n1 = FunctionInput() {inputType=FunctionInput_MaterialAttributes}
  n2 = GetMaterialAttributes(None=@n1) {attributeGetTypes [#1, #2]}
  n3 = VectorParameter() {group=BlastDecal, parameterName=BlastColor, defaultValue=<rgba 1 1 1 1>}
  n4 = NamedRerouteUsage()
  n5 = ComponentMask(None=@n4) {b=true, g=true, r=true}
  n6 = ComponentMask(None=@n4) {a=true}
  n7 = LinearInterpolate(
   A=@n2.BaseColor
   B=Multiply(A=Multiply(A=@n3.RGB, B=@n3.RGB), B=Multiply(A=@n2.BaseColor, B=@n5))
   Alpha=Multiply(A=@n6, B=ScalarParameter() {defaultValue=1, group=BlastDecal, parameterName=BlendBlastColor})
  )
  n8 = LinearInterpolate(
   A=@n2.Specular
   B=ScalarParameter() {defaultValue=0.25, group=BlastDecal, parameterName=BlastSpecular}
   Alpha=Multiply(A=@n6, B=ScalarParameter() {defaultValue=1, group=BlastDecal, parameterName=BlendBlastSpecular})
  )
  FunctionOutput(None=StaticSwitchParameter(
    True=SetMaterialAttributes!(MaterialAttributes=@n2.MaterialAttributes, BaseColor=@n7, Specular=@n8) {attributeSetTypes [#1, #2]}
    False=@n1
   ) {group=BlastDecal, parameterName=UseBlastDecal})
  n9 = VectorParameter() {group=BlastDecal, parameterName=BlastOffsetAndScale, channelNames {a=ScaleV, b=ScaleU, g=OffsetV, r=OffsetU}, defaultValue=<rgba 0 0 1 1>}
  n10 = MaterialFunctionCall(
   UVs=Add(A=TextureCoordinate() {coordinateIndex=2}, B=ComponentMask(None=@n9) {g=true, r=true})
   Texture Scale=AppendVector(A=@n9.ScaleU, B=@n9.ScaleU)
  ) {materialFunction=&ScaleUVsByCenter}
  n11 = TextureSampleParameter2D(UVs=@n10.UVs) {
   group=BlastDecal
   parameterName=BlastDecalTexture
   samplerSource=SSM_Clamp_WorldGroupSettings
   texture=/Game/Global/Textures/Default/Blanks/T_Blank_LegalWhite_sRGBColor
  }
  n12 = TextureSampleParameter2D(UVs=@n10.UVs) {
   group=BlastDecal
   parameterName=BlastDecalTexture
   samplerSource=SSM_Wrap_WorldGroupSettings
   texture=&T_Blank_LegalWhite_sRGBColor
  }
  NamedRerouteDeclaration(None=StaticSwitchParameter(True=@n11.RGBA, False=@n12.RGBA) {group=BlastDecal, parameterName=UseClampedUVs}) {name=Texture_Route, nodeColor=<rgba 0.4297 1 0 1>}
 ]}
